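# StateTree for the RoninDemon quest scenario, running a single-wave assault encounter from player arrival to completion.
# Init waits for a player count above zero, then Stage_Wave_One fetches the CubicVice and PolishedJade actors, adding a friendly icon and outgoing-damage-reduction effects.
# It spawns assault henchmen at tagged points, sets FirePetal grunt alert targets, and advances when the wave count runs out.
# Complete hides the icon, sends the encounter-completed verb and succeeds.
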StateTree /BattlePassS33_Quests/Items/QuestItems/Scenario/RoninDemon/ST_QuestScenario_RoninDemon
states [
 Root {
  id=#1
  type=STATE
  children [#2, #3, #4]
  tasks []
  transitions []
 }
 Init {
  id=#2
  type=STATE
  tasks [{
    id=#5
    description="Scenario [Player Count Transition]"
    properties {
     iD=#6
     instance {
      _structType=/Script/EncountersRuntime.EncounterPlayerCountTransitionStateTreeTaskInstanceData
      bOnlyTriggerOnce=true
      bShouldTransition=true
      bTriggered=false
      comparisonOperator=GreaterThan
      userActor=null
      value=0
      customVerb {amount=1, customVerbTag {tagName=None}, subjects [], values []}
      onTriggered {iD=#7}
     }
     node {
      name=None
      _structType=/Script/EncountersRuntime.EncounterPlayerCountTransitionStateTreeTask
      bCanEditConsideredForCompletion=true
      bConsideredForCompletion=true
      bHasTaskCompletionDelegateDispatcher=false
      bTaskEnabled=true
      transitionHandlingPriority=Normal
      bindingsBatch {value=65535}
      executionRuntimeTemplateIndex {value=65535}
      instanceDataHandle {index=65535, source=None, stateHandle {index=65535}}
      instanceTemplateIndex {value=65535}
      outputBindingsBatch {value=65535}
      transitionTo {name=None, fallback=None, iD=#7, linkType=NextSelectableState, stateHandle {index=65535}}
     }
    }
   }]
  transitions []
 }
 Stage_Wave_One {
  id=#3
  type=STATE
  tasks [
   {
    id=#8
    description=Get Encounter Actor - CubicVice
    properties {
     iD=#9
     instance {
      _structType=&GetEncounterActorStateTreeTaskInstanceData
      actorQuery {autoDescription=" ANY( AISpawnerData.Type.Tandem.CubicVice )", tagDictionary [{tagName=AISpawnerData.Type.Tandem.CubicVice}]}
     }
     node {name=Get Encounter Actor - CubicVice, _structType=&GetEncounterActorStateTreeTask}
    }
   }
   {
    id=#10
    description=Grant Reactions - Friendly Icon On
    properties {
     iD=#11
     instance {
      _structType=/Script/EncountersRuntime.GrantReactionStateTreeTaskInstanceData
      grantType=GrantToTargetActor
      targetActor=null
      userActor=null
      reactionEntries [{rowValue {
         bUseDataRegistryId=false
         dataRegistryId=""
         dataTableRow {dataTable=&/DT_QuestScenario_RoninDemon_Reactions, rowName=Icon_Visibility_on}
        }}]
      reactions []
     }
     node {
      name=Grant Reactions - Friendly Icon On
      _structType=/Script/EncountersRuntime.GrantReactionStateTreeTask
      bCanEditConsideredForCompletion=true
      bConsideredForCompletion=true
      bHasTaskCompletionDelegateDispatcher=false
      bTaskEnabled=true
      transitionHandlingPriority=Normal
      bindingsBatch {value=65535}
      executionRuntimeTemplateIndex {value=65535}
      instanceDataHandle {index=65535, source=None, stateHandle {index=65535}}
      instanceTemplateIndex {value=65535}
      outputBindingsBatch {value=65535}
     }
    }
   }
   {
    id=#12
    description=Get Encounter Actor - PolishedJade
    properties {
     iD=#13
     instance {
      _structType=&GetEncounterActorStateTreeTaskInstanceData
      bQueryAgainstTrackedActorIdentifiers=true
      actorQuery {autoDescription=" ANY( AISpawnerData.Type.Tandem.PolishedJade )", tagDictionary [{tagName=AISpawnerData.Type.Tandem.PolishedJade}]}
     }
     node {name=Get Encounter Actor - PolishedJade, _structType=&GetEncounterActorStateTreeTask}
    }
   }
   {
    id=#14
    description="LWM Encounter [Add Gameplay Effect] - CubicVice"
    properties {
     iD=#15
     instance {
      _structType=/Script/EncountersRuntime.LWMEncounterAddGameplayEffectStateTreeTaskInstanceData
      bRemoveEffectOnExit=false
      gameplayEffect=&/GE_OutgoingDamageReduction.GE_OutgoingDamageReduction_C
      userActor=null
      encounterQuery {autoDescription="", tokenStreamVersion=0, userDescription="", queryTokenStream [], tagDictionary []}
      identifierTags {gameplayTags []}
     }
     node {
      name="LWM Encounter [Add Gameplay Effect] - CubicVice"
      _structType=/Script/EncountersRuntime.LWMEncounterAddGameplayEffectStateTreeTask
      bCanEditConsideredForCompletion=true
      bConsideredForCompletion=true
      bHasTaskCompletionDelegateDispatcher=false
      bTaskEnabled=true
      transitionHandlingPriority=Normal
      bindingsBatch {value=65535}
      executionRuntimeTemplateIndex {value=65535}
      instanceDataHandle {index=65535, source=None, stateHandle {index=65535}}
      instanceTemplateIndex {value=65535}
      outputBindingsBatch {value=65535}
     }
    }
   }
   {
    id=#16
    description="LWM Encounter [Add Gameplay Effect] - PolishedJade"
    properties {
     iD=#17
     instance {
      _structType=&LWMEncounterAddGameplayEffectStateTreeTaskInstanceData
      bRemoveEffectOnExit=false
      gameplayEffect=&GE_OutgoingDamageReduction_C
      userActor=null
      encounterQuery {autoDescription="", tokenStreamVersion=0, userDescription="", queryTokenStream [], tagDictionary []}
      identifierTags {gameplayTags []}
     }
     node {
      name="LWM Encounter [Add Gameplay Effect] - PolishedJade"
      _structType=&LWMEncounterAddGameplayEffectStateTreeTask
      bCanEditConsideredForCompletion=true
      bConsideredForCompletion=true
      bHasTaskCompletionDelegateDispatcher=false
      bTaskEnabled=true
      transitionHandlingPriority=Normal
      bindingsBatch {value=65535}
      executionRuntimeTemplateIndex {value=65535}
      instanceDataHandle {index=65535, source=None, stateHandle {index=65535}}
      instanceTemplateIndex {value=65535}
      outputBindingsBatch {value=65535}
     }
    }
   }
   {
    id=#18
    description=Spawn Wave One
    properties {
     iD=#19
     instance {
      _structType=/Script/EncountersRuntime.LWMEncounterDynamicStartStateTreeTaskInstanceData
      bAttachVolumeToSpawn=false
      bOverrideAnchorPosition=true
      bOverrideDelayBetweenEvents=false
      bOverrideEnvQuery=false
      bOverrideStageActorMaxCount=false
      bOverrideStageActorMaxSpawnedCount=false
      bPauseOnExit=false
      bSpecifyFaction=true
      bUseVolumeAsAnchor=false
      bWaitForEncounterDoneSpawning=false
      delayBetweenEventsMax=0
      delayBetweenEventsMin=0
      dynamicEmptyEncounter=/FortQuestEncounters/Lager/Lager_Encounter_Empty_QuestEncounter
      faction=&/DA_RoninDemon_Faction
      instigator=null
      owningActor=null
      spawnAroundEnvironmentQuery=null
      spawnAroundEnvironmentQueryRunMode=AllMatching
      spawnAroundNavigationSourceOverride=""
      stageActorMaxCount=1
      stageActorMaxSpawnedCount=-1
      userActor=null
      actorTags {gameplayTags []}
      anchorPointTagQuery {
       autoDescription=" ANY( QuestEncounter.RoninDemon.Anchor.Enemy_1_1 )"
       tokenStreamVersion=0
       userDescription=""
       queryTokenStream [0,1,1,1,0]
       tagDictionary [{tagName=QuestEncounter.RoninDemon.Anchor.Enemy_1_1}]
      }
      dynamicEncounterEntries [{
        actorCountOverrideMax=1
        actorCountOverrideMin=1
        bCreateRandomSingleActorEvents=false
        bOverrideActorCount=false
        bUseEncounterTagQuery=true
        bUseEventTemplate=false
        debugName=""
        maxConcurrentSpawns=2
        noActorLimitEventCount=1
        spawnCount=2
        actorDescriptions []
        encounterTagQuery {
         autoDescription=" ANY( Athena.Quest.Scenario.Faction.Henchmen.Assault )"
         tokenStreamVersion=0
         userDescription=""
         queryTokenStream [0,1,1,1,0]
         tagDictionary [{tagName=Athena.Quest.Scenario.Faction.Henchmen.Assault}]
        }
        eventTemplate {
         activationGamePhaseStep=None
         activationMaxDelay=0
         activationMinDelay=0
         activationSafeZoneIndex=0
         bCanSpawnWithoutMatchingPlayerDistance=false
         bCheckForActorDensity=true
         bOnlyInNextSafeZone=false
         bOnlyInNextSafeZoneIfStormShrinking=false
         bOnlyInSafeZone=true
         bShouldProcessEventRequestImmediately=false
         deactivationMaxDelay=-1
         deactivationMinDelay=-1
         maxActorDensityOverride=-1
         maxDistanceToPlayer=-1
         minDistanceToPlayer=-1
         randomActivationChance=100
         weight=1
         calendarEventConditions {isActiveWithoutSeasonalManager=false, conditions []}
         canSpawnCondition {description=Empty, schemaClass=/Script/LagerRuntime.FortAthenaLivingWorldConditionSchema, editableConditions [], sharedDefinition {}}
         despawnCondition {
          bCanBeDespawned=false
          bCanDespawnInTheStorm=false
          bCanDespawnWhenInCombatWithPlayer=false
          canStartDespawningPhaseStep=None
          canStartDespawningSafeZonePhase=-1
          maxStimulusAgeToBeConsideredInCombat=30
          minDistanceToPlayer=-1
          timeAsCandidateBeforeDespawn=10
         }
         despawnConditions []
         eventDatas []
         eventInstanceMaxCount [-1]
         eventInstanceMaxSpawnedCount [-1]
         instantDeathMaxRespawnTime []
         instantDeathMinRespawnTime []
         isEnabled {registryType=None, value=1, curve {curveTable=null, rowName=None}}
         maxCount [0]
         maxRespawnTime [45]
         maxSpawnedCount [-1]
         minRespawnTime [30]
         playlistTagQuery {autoDescription="", tokenStreamVersion=0, userDescription="", queryTokenStream [], tagDictionary []}
         tagDensityConditions []
        }
        pointProviderTagQueryOverride {autoDescription="", tokenStreamVersion=0, userDescription="", queryTokenStream [], tagDictionary []}
       }]
      encounterIdentifierTags {gameplayTags [{tagName=QuestEncounter.RoninDemon.AI.Goal.Assault.Wave.1}]}
      leashRadiusInner {registryType=None, value=200, curve {curveTable=null, rowName=None}}
      leashRadiusOuter {registryType=None, value=800, curve {curveTable=null, rowName=None}}
      pointProviderTagQuery {
       autoDescription=""
       tokenStreamVersion=0
       userDescription=""
       queryTokenStream [0,1,1,1,0]
       tagDictionary [{tagName=QuestEncounter.RoninDemon.Location.SpawnPoint.Wave.1}]
      }
      spawnActions [{spawnActionClass=null, spawnActionTag {tagName=None}}]
      spawnTimeout {registryType=None, value=-1, curve {curveTable=null, rowName=None}}
     }
     node {
      name=Spawn Wave One
      _structType=/Script/EncountersRuntime.LWMEncounterDynamicStartStateTreeTask
      bCanEditConsideredForCompletion=true
      bConsideredForCompletion=true
      bHasTaskCompletionDelegateDispatcher=false
      bTaskEnabled=true
      transitionHandlingPriority=Normal
      bindingsBatch {value=65535}
      executionRuntimeTemplateIndex {value=65535}
      instanceDataHandle {index=65535, source=None, stateHandle {index=65535}}
      instanceTemplateIndex {value=65535}
      outputBindingsBatch {value=65535}
     }
    }
   }
   {
    id=#20
    description=Demon Set Target
    properties {
     iD=#21
     instance {
      _structType=/Script/EncountersRuntime.LWMEncounterSetAlertLevelStateTreeTaskInstanceData
      targetActor=null
      targetBehavior=OneTime
      userActor=null
      encounterQuery {
       autoDescription=" ANY( AISpawnerData.Type.Guard.FirePetal.Grunt )"
       tokenStreamVersion=0
       userDescription=""
       queryTokenStream [0,1,1,1,0]
       tagDictionary [{tagName=AISpawnerData.Type.Guard.FirePetal.Grunt}]
      }
      targetEncounterQuery {autoDescription="", tokenStreamVersion=0, userDescription="", queryTokenStream [], tagDictionary []}
      targetInfo {awareForgetTime=15}
     }
     node {
      name=Demon Set Target
      _structType=/Script/EncountersRuntime.LWMEncounterSetAlertLevelStateTreeTask
      bCanEditConsideredForCompletion=true
      bConsideredForCompletion=true
      bHasTaskCompletionDelegateDispatcher=false
      bTaskEnabled=true
      transitionHandlingPriority=Normal
      bindingsBatch {value=65535}
      executionRuntimeTemplateIndex {value=65535}
      instanceDataHandle {index=65535, source=None, stateHandle {index=65535}}
      instanceTemplateIndex {value=65535}
      outputBindingsBatch {value=65535}
     }
    }
   }
   {
    id=#22
    description="LWM Encounter [Remaining Count Transition]"
    properties {
     iD=#23
     instance {
      _structType=/Script/EncountersRuntime.LWMEncounterRemainingCountTransitionStateTreeTaskInstanceData
      bOnlyTriggerOnce=true
      bShouldTransition=true
      bTriggered=false
      userActor=null
      value=0
      customVerb {amount=1, customVerbTag {tagName=None}, subjects [], values []}
      encounterQuery {
       autoDescription=""
       tokenStreamVersion=0
       userDescription=""
       queryTokenStream [0,1,1,1,0]
       tagDictionary [{tagName=QuestEncounter.RoninDemon.AI.Goal.Assault.Wave.1}]
      }
      onTriggered {iD=#7}
     }
     node {
      name=None
      _structType=/Script/EncountersRuntime.LWMEncounterRemainingCountTransitionStateTreeTask
      bCanEditConsideredForCompletion=true
      bConsideredForCompletion=true
      bHasTaskCompletionDelegateDispatcher=false
      bTaskEnabled=true
      transitionHandlingPriority=Normal
      bindingsBatch {value=65535}
      executionRuntimeTemplateIndex {value=65535}
      instanceDataHandle {index=65535, source=None, stateHandle {index=65535}}
      instanceTemplateIndex {value=65535}
      outputBindingsBatch {value=65535}
      transitionTo {name=None, fallback=None, iD=#7, linkType=NextSelectableState, stateHandle {index=65535}}
     }
    }
   }
  ]
  transitions []
 }
 Complete {
  id=#4
  type=STATE
  tasks [
   {
    id=#24
    description=Get Scenario Actor - CubicVice
    properties {
     iD=#25
     instance {
      _structType=&GetEncounterActorStateTreeTaskInstanceData
      bQueryAgainstTrackedActorIdentifiers=true
      actorQuery {autoDescription=" ANY( AISpawnerData.Type.Tandem.CubicVice )", tagDictionary [{tagName=AISpawnerData.Type.Tandem.CubicVice}]}
     }
     node {name=Get Scenario Actor - CubicVice, _structType=&GetEncounterActorStateTreeTask}
    }
   }
   {
    id=#26
    description=Grant Reactions - Friendly Icon Off
    properties {
     iD=#27
     instance {
      _structType=&GrantReactionStateTreeTaskInstanceData
      grantType=GrantToTargetActor
      targetActor=null
      userActor=null
      reactionEntries [{rowValue {
         bUseDataRegistryId=false
         dataRegistryId=""
         dataTableRow {dataTable=&DT_QuestScenario_RoninDemon_Reactions, rowName=Icon_Visibility_off}
        }}]
      reactions []
     }
     node {
      name=Grant Reactions - Friendly Icon Off
      _structType=&GrantReactionStateTreeTask
      bCanEditConsideredForCompletion=true
      bConsideredForCompletion=true
      bHasTaskCompletionDelegateDispatcher=false
      bTaskEnabled=true
      transitionHandlingPriority=Normal
      bindingsBatch {value=65535}
      executionRuntimeTemplateIndex {value=65535}
      instanceDataHandle {index=65535, source=None, stateHandle {index=65535}}
      instanceTemplateIndex {value=65535}
      outputBindingsBatch {value=65535}
     }
    }
   }
   {
    id=#28
    description=Send Custom Verb
    properties {
     iD=#29
     instance {
      _structType=/Script/EncountersRuntime.EncounterSendCustomVerbStateTreeTaskInstanceData
      bSendVerbForEachPlayerAuthor=false
      userActor=null
      customVerb {amount=1, customVerbTag {tagName=CustomVerb.Encounter.Completed}, subjects [], values []}
     }
     node {
      name=None
      _structType=/Script/EncountersRuntime.EncounterSendCustomVerbStateTreeTask
      bCanEditConsideredForCompletion=true
      bConsideredForCompletion=true
      bHasTaskCompletionDelegateDispatcher=false
      bTaskEnabled=true
      transitionHandlingPriority=Normal
      bindingsBatch {value=65535}
      executionRuntimeTemplateIndex {value=65535}
      instanceDataHandle {index=65535, source=None, stateHandle {index=65535}}
      instanceTemplateIndex {value=65535}
      outputBindingsBatch {value=65535}
     }
    }
   }
   {
    id=#30
    description=BP SDA STTask Complete State
    properties {
     iD=#31
     instanceObject=&self:StateTreeEditorData_0.StateTreeState_7.BP_SDA_STTask_CompleteState_C_0
     node {
      name=None
      _structType=&StateTreeBlueprintTaskWrapper
      taskClass=/StateDrivenActor/Tasks/BP_SDA_STTask_CompleteState.BP_SDA_STTask_CompleteState_C
     }
    }
   }
  ]
  transitions [{id=#32, trigger=ON_STATE_COMPLETED, state {id=#33, name=Root, linkType=SUCCEEDED}}]
 }
]
